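# Sound cue for the warning tell of a lightning storm, layering electricity with thunder.
# A Mixer blends a louder, modulated electricity wave with a quieter, equally weighted random pick of three thunder waves.
# The thunder branch has its own attenuation asset, while the cue overrides attenuation for long-range playback.
# It plays through the Fortnite storms sound class.
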SoundCue /SunRoseZeusGameplay/Lightning/Audio/Lightning_Storm_Tell_Cue
root=Mixer_0
nodes [
 {id=Attenuation_0, children [Random_0], properties {attenuationSettings=&/LIghtning_Thunder_Tell_Att}}
 {id=Mixer_0, children [Modulator_0, Attenuation_0], properties {inputVolume [2,0.5]}}
 {id=Modulator_0, children [WavePlayer_2]}
 {id=Random_0, children [WavePlayer_3,WavePlayer_4,WavePlayer_5], properties {weights [1,1,1]}}
 {id=WavePlayer_2, properties {soundWaveAssetPtr=&/Lightning_Tell_Electricity_01}}
 {id=WavePlayer_3, properties {soundWaveAssetPtr=&/Lightning_Tell_Thunder_01}}
 {id=WavePlayer_4, properties {soundWaveAssetPtr=&/Lightning_Tell_Thunder_02}}
 {id=WavePlayer_5, properties {soundWaveAssetPtr=&/Lightning_Tell_Thunder_03}}
]
properties {
 bOverrideAttenuation=true
 maxDistance=32000
 soundClassObject=/Game/Sounds/FortSoundClassesAndModes/Fort_Storms_PSM
 volumeMultiplier=1
 attenuationOverrides {falloffDistance=3000, stereoSpread=5000, attenuationShapeExtents {x=1000}}
}
